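; Passive gameplay ability that grants a quest, named for Season 33 SMG Expertise, to the actor that owns it.
; On activation it calls the parent, then validates the owning actor.
; If valid, it passes the quest definition variable and that actor to the QuestInstanceManagerSubsystem's GrantQuest, then ends itself.
; An invalid owner leads to no grant and no EndAbility call in this graph.
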(event Ability|EventFailedToActivatePassiveAbility
  (params (ActorInfo :type Gameplay%20Ability%20Actor%20Info%20Structure))
  (|CallParentFunction ActorInfo))

(event Ability|EventActivateAbility
  (bind _returnvalue (Ability|GetOwningActorfromActorInfo))
  (|CallParentFunction)
  (Utilities|IsValid _returnvalue
    (:Is%20Valid
      (Class|QuestInstanceManagerSubsystem|GrantQuest (WorldSubsystems|GetQuestInstanceManagerSubsystem) (Utilities|Struct|MakeGrantQuestBlueprintParameters (Variables|Default|GetQuest%20Defenition) _returnvalue true))
      (Ability|EndAbility))))

(event Ability|EventOnEndAbility
  (params (bWasCancelled :type Boolean)))
